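# Physics asset for the Winter Holiday backpack accessory, giving its dangling sword, grenades, bow and foliage simple collision bodies and joints for secondary motion.
# The sword, both grenades, bow center and two foliage pieces attach directly to backpack_root, using box or capsule shapes.
# The hoop chains off the sword, while the bow's side loops and two-segment tails branch from its center.
# Foliage joints swing and twist freely; the others carry per-joint angular limits.
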PhysicsAsset /Game/Accessories/FORT_Backpacks/F_MED_Winter_Holiday/Meshes/F_MED_Winter_Holiday_Physics
bodies {
 backpack_root [{shapeType=Capsule, length=18, radius=17, center=<-29 6.8 -1.7>, rotation=<rot 0 90 0>}]
 dyn_bow_btm_l_01 [{
   shapeType=Capsule
   length=2.5
   radius=2.5
   shapeName=dyn_bow_btm_l_01_capsule
   center=<0.2618 0 0.0451>
   rotation=<rot 80.2344 -179.9996 -179.9999>
  }]
 dyn_bow_btm_l_02 [{
   shapeType=Capsule
   length=2.5
   radius=2.5
   shapeName=dyn_bow_btm_l_02_capsule
   center=<0.1031 0.0234 -0.5563>
   rotation=<rot 10.5601 1.695 -2.0555>
  }]
 dyn_bow_btm_r_01 [{shapeType=Capsule, length=2.5, radius=2.5, shapeName=dyn_bow_btm_r_01_capsule, center=<-0.7665 0 0>, rotation=<rot 90 0 0>}]
 dyn_bow_btm_r_02 [{
   shapeType=Capsule
   length=2.5
   radius=2.5
   shapeName=dyn_bow_btm_r_02_capsule
   center=<0 0.0408 -0.6262>
   rotation=<rot 5.2136 -0.3409 -3.7471>
  }]
 dyn_bow_c [{
   shapeType=Box
   extentX=2
   extentY=3.8523
   extentZ=4.328
   shapeName=dyn_bow_c_box
   center=<0.0039 -0.0123 -0.2368>
   rotation=<rot 0.0397 -75.4414 -13.3625>
  }]
 dyn_bow_l [{
   shapeType=Box
   extentX=2.7472
   extentY=6.4767
   extentZ=6.3239
   shapeName=dyn_bow_l_box
   center=<-1.2992 1.0936 1.5783>
   rotation=<rot -7.7631 -99.4037 -1.5493>
  }]
 dyn_bow_r [{
   shapeType=Box
   extentX=2.54
   extentY=6.9912
   extentZ=5.7909
   shapeName=dyn_bow_r_box
   center=<1.6226 0.2133 1.4068>
   rotation=<rot 4.1939 -61.2272 -13.0961>
  }]
 dyn_foliage_l [{
   shapeType=Capsule
   length=12.76
   radius=3.5
   shapeName=dyn_foliage_l_capsule
   center=<-2.6126 0.2123 -6.1227>
   rotation=<rot 1.0824 -89.2895 33.2621>
  }]
 dyn_foliage_r [{
   shapeType=Capsule
   length=13.51
   radius=3.5
   shapeName=dyn_foliage_r_capsule
   center=<1.984 1.458 -6.8616>
   rotation=<rot 3.7586 -93.5786 -30.1147>
  }]
 dyn_grenade_l [{shapeType=Capsule, length=0.25, radius=2.6692, center=<0.1702 -3.2114 -0.6882>, rotation=<rot 20.9222 -89.9999 -8.6415>}]
 dyn_grenade_r [{shapeType=Capsule, length=0, radius=3.6604, center=<-0.5704 -1.7244 -1.2698>, rotation=<rot 22.9941 -90 12.138>}]
 dyn_sword [{
   shapeType=Box
   extentX=2.3345
   extentY=9.0313
   extentZ=52.4786
   shapeName=dyn_sword_box
   center=<-2.8499 0.0002 -2.6639>
   rotation=<rot 0 90.0979 -47.8004>
  }]
 dyn_sword_hoop [{shapeType=Capsule, length=1.3378, radius=1.5, center=<0.5633 0.146 -2.8664>, rotation=<rot 79.4316 173.5207 -94.5384>}]
}
constraints [
 {bone1Name=dyn_sword, bone2Name=backpack_root, swing1LimitDegrees=45, swing2LimitDegrees=45, twistLimitDegrees=45}
 {bone1Name=dyn_sword_hoop, bone2Name=dyn_sword, swing1LimitDegrees=20, swing2LimitDegrees=30, twistLimitDegrees=0}
 {bone1Name=dyn_grenade_l, bone2Name=backpack_root, swing1LimitDegrees=45, swing2LimitDegrees=45, twistLimitDegrees=10}
 {bone1Name=dyn_grenade_r, bone2Name=backpack_root, swing1LimitDegrees=45, swing2LimitDegrees=45, twistLimitDegrees=10}
 {bone1Name=dyn_bow_l, bone2Name=dyn_bow_c, swing1LimitDegrees=45, swing2LimitDegrees=45, twistLimitDegrees=45}
 {bone1Name=dyn_bow_r, bone2Name=dyn_bow_c, swing1LimitDegrees=45, swing2LimitDegrees=45, twistLimitDegrees=45}
 {bone1Name=dyn_bow_c, bone2Name=backpack_root, swing1LimitDegrees=45, swing2LimitDegrees=45, twistLimitDegrees=45}
 {bone1Name=dyn_bow_btm_l_01, bone2Name=dyn_bow_c, swing1LimitDegrees=90, swing2LimitDegrees=60.84, twistLimitDegrees=81}
 {bone1Name=dyn_bow_btm_l_02, bone2Name=dyn_bow_btm_l_01, swing1LimitDegrees=57.96, swing2LimitDegrees=52.2, twistLimitDegrees=85.32}
 {bone1Name=dyn_bow_btm_r_01, bone2Name=dyn_bow_c, swing1LimitDegrees=72.72, swing2LimitDegrees=64.08, twistLimitDegrees=77.04}
 {bone1Name=dyn_bow_btm_r_02, bone2Name=dyn_bow_btm_r_01, swing1LimitDegrees=71.28, swing2LimitDegrees=74.16, twistLimitDegrees=77.04}
 {
  bone1Name=dyn_foliage_r
  bone2Name=backpack_root
  swing1LimitDegrees=45
  swing1Motion=Free
  swing2LimitDegrees=45
  swing2Motion=Free
  twistLimitDegrees=45
  twistMotion=Free
 }
 {
  bone1Name=dyn_foliage_l
  bone2Name=backpack_root
  swing1LimitDegrees=45
  swing1Motion=Free
  swing2LimitDegrees=45
  swing2Motion=Free
  twistLimitDegrees=45
  twistMotion=Free
 }
]
