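# Animation Blueprint for the F_MED_Winter_Holiday backpack cosmetic, adding secondary motion to its bows and foliage.
# From the reference pose, a RigidBody node simulates the override physics asset relative to backpack_root with custom gravity.
# Partially blended spring AnimDynamics nodes then move upper bow and foliage bones, before additive bone transforms offset the center and bottom bows.
# Separate AnimDynamics chains for bottom bows and grenades are unconnected to the output; the event graph does nothing.
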AnimBlueprint /BRCosmetics/Accessories/FORT_Backpacks/F_MED_Winter_Holiday/Meshes/F_MED_Winter_Holiday_Skeleton_AnimBP
target_skeleton=/Game/Accessories/FORT_Backpacks/F_MED_Winter_Holiday/Meshes/F_MED_Winter_Holiday_Skeleton
graphs [
 AnimGraph {nodes [
   n1 = Anim Dynamics - Bone: dyn_bow_btm_r_01(Alpha=0.7, ExternalForce="0,0,0") {
    angularSpringConstant=50
    bAngularSpring=true
    bGravityOverrideInSimSpace=true
    bLinearSpring=true
    linearSpringConstant=25
    numSolverIterationsPostUpdate=4
    numSolverIterationsPreUpdate=16
    alphaBoolBlend {blendOption=Linear}
    boundBone {boneName=dyn_bow_btm_r_01}
    componentAppliedLinearAccClamp=<2500 2500 2500>
    componentLinearAccScale=<1 1 1>
    componentLinearVelScale=<1 1 1>
    physicsBodyDefinitions [{
      sphereCollisionRadius=0
      boundBone {boneName=dyn_bow_btm_r_01}
      boxExtents=<2 5 10>
      constraintSetup {angularLimitsMax=<15 40 10>, angularLimitsMin=<15 -20 -50>, angularTarget=<0 0 20>, linearAxesMax=<5 1 2.5>, linearAxesMin=<-1.5 -5 -5>}
      localJointOffset=<0 0 -0.5>
     }]
   }
   n2 = Anim Dynamics - Bone: dyn_bow_btm_r_02(ComponentPose=@n1, Alpha=0.7, ExternalForce="0,0,0") {
    angularSpringConstant=50
    bAngularSpring=true
    bGravityOverrideInSimSpace=true
    bLinearSpring=true
    linearSpringConstant=50
    numSolverIterationsPostUpdate=4
    numSolverIterationsPreUpdate=16
    alphaBoolBlend {blendOption=Linear}
    boundBone {boneName=dyn_bow_btm_r_02}
    componentAppliedLinearAccClamp=<2500 2500 2500>
    componentLinearAccScale=<1 1 1>
    componentLinearVelScale=<1 1 1>
    physicsBodyDefinitions [{
      sphereCollisionRadius=0
      boundBone {boneName=dyn_bow_btm_r_02}
      boxExtents=<5 5 10>
      constraintSetup {angularLimitsMax=<50 50 20>, angularLimitsMin=<-50 -50 20>, angularTarget=<0 0 0>, linearAxesMax=<5 2 5>, linearAxesMin=<-5 -5 0>}
      localJointOffset=<0 0 -0.5>
     }]
   }
   n3 = Anim Dynamics - Bone: dyn_grenade_l(ComponentPose=@n2, Alpha=0.3, ExternalForce="0,0,0") {
    angularSpringConstant=50
    bAngularSpring=true
    bGravityOverrideInSimSpace=true
    bLinearSpring=true
    linearSpringConstant=75
    numSolverIterationsPostUpdate=4
    numSolverIterationsPreUpdate=16
    alphaBoolBlend {blendOption=Linear}
    boundBone {boneName=dyn_grenade_l}
    componentAppliedLinearAccClamp=<2500 2500 2500>
    componentLinearAccScale=<1 1 1>
    componentLinearVelScale=<1 1 1>
    physicsBodyDefinitions [{
      sphereCollisionRadius=0
      boundBone {boneName=dyn_grenade_l}
      boxExtents=<2 2 2>
      constraintSetup {angularLimitsMax=<2 2 1>, angularLimitsMin=<-1 -5 -1>, angularTarget=<0 0 0>, linearAxesMax=<0 1 0>, linearAxesMin=<0 -0.5 -1>}
      localJointOffset=<0 -1 -2>
     }]
   }
   Anim Dynamics - Bone: dyn_grenade_r(ComponentPose=@n3, Alpha=0.4, ExternalForce="0,0,0") {
    angularSpringConstant=50
    bAngularSpring=true
    bGravityOverrideInSimSpace=true
    bLinearSpring=true
    linearSpringConstant=75
    numSolverIterationsPostUpdate=4
    numSolverIterationsPreUpdate=16
    alphaBoolBlend {blendOption=Linear}
    boundBone {boneName=dyn_grenade_r}
    componentAppliedLinearAccClamp=<2500 2500 2500>
    componentLinearAccScale=<1 1 1>
    componentLinearVelScale=<1 1 1>
    physicsBodyDefinitions [{
      sphereCollisionRadius=0
      boundBone {boneName=dyn_grenade_r}
      boxExtents=<5 5 5>
      constraintSetup {angularLimitsMax=<2 2 1>, angularLimitsMin=<-1 -1 -1>, angularTarget=<0 0 0>, linearAxesMax=<2 1 1>, linearAxesMin=<-2 -1.5 -2>}
      localJointOffset=<0 0 -2>
     }]
   }
   n4 = Anim Dynamics - Bone: dyn_bow_btm_l_01(Alpha=0.7, ExternalForce="0,0,0") {
    angularSpringConstant=50
    bAngularSpring=true
    bGravityOverrideInSimSpace=true
    bLinearSpring=true
    linearSpringConstant=25
    numSolverIterationsPostUpdate=4
    numSolverIterationsPreUpdate=16
    alphaBoolBlend {blendOption=Linear}
    boundBone {boneName=dyn_bow_btm_l_01}
    componentAppliedLinearAccClamp=<2500 2500 2500>
    componentLinearAccScale=<1 1 1>
    componentLinearVelScale=<1 1 1>
    physicsBodyDefinitions [{
      sphereCollisionRadius=0
      boundBone {boneName=dyn_bow_btm_l_01}
      boxExtents=<2 5 10>
      constraintSetup {angularLimitsMax=<15 20 10>, angularLimitsMin=<15 -20 -50>, angularTarget=<0 0 20>, linearAxesMax=<5 1 5>, linearAxesMin=<-1 -5 -2>}
      localJointOffset=<0 0 -0.5>
     }]
   }
   Anim Dynamics - Bone: dyn_bow_btm_l_02(ComponentPose=@n4, Alpha=0.7, ExternalForce="0,0,0") {
    angularSpringConstant=50
    bAngularSpring=true
    bGravityOverrideInSimSpace=true
    bLinearSpring=true
    linearSpringConstant=50
    numSolverIterationsPostUpdate=4
    numSolverIterationsPreUpdate=16
    alphaBoolBlend {blendOption=Linear}
    boundBone {boneName=dyn_bow_btm_l_02}
    componentAppliedLinearAccClamp=<2500 2500 2500>
    componentLinearAccScale=<1 1 1>
    componentLinearVelScale=<1 1 1>
    physicsBodyDefinitions [{
      sphereCollisionRadius=0
      boundBone {boneName=dyn_bow_btm_l_02}
      boxExtents=<5 5 10>
      constraintSetup {angularLimitsMax=<50 50 20>, angularLimitsMin=<-50 -50 20>, angularTarget=<0 0 0>, linearAxesMax=<5 2 5>, linearAxesMin=<-5 -5 0>}
      localJointOffset=<0 0 -0.5>
     }]
   }
   n5 = RigidBody(ComponentPose=Local To Component(LocalPose=Local Space Ref Pose())) {
    bOverrideWorldGravity=true
    overridePhysicsAsset=/Game/Accessories/FORT_Backpacks/F_MED_Winter_Holiday/Meshes/F_MED_Winter_Holiday_Physics
    simulationSpace=BaseBoneSpace
    alphaBoolBlend {blendOption=Linear}
    baseBoneRef {boneName=backpack_root}
    componentAppliedLinearAccClamp=<1000 1000 1000>
    componentLinearAccScale=<1 1 1>
    componentLinearVelScale=<1 1 1>
    overrideWorldGravity {z=-1100}
   }
   n6 = Anim Dynamics - Bone: dyn_bow_c(ComponentPose=@n5, Alpha=0.5, ExternalForce="0,0,0") {
    angularSpringConstant=20
    bAngularSpring=true
    bGravityOverrideInSimSpace=true
    bLinearSpring=true
    linearSpringConstant=20
    numSolverIterationsPostUpdate=4
    numSolverIterationsPreUpdate=16
    alphaBoolBlend {blendOption=Linear}
    boundBone {boneName=dyn_bow_c}
    componentAppliedLinearAccClamp=<2500 2500 2500>
    componentLinearAccScale=<1 1 1>
    componentLinearVelScale=<1 1 1>
    physicsBodyDefinitions [{
      sphereCollisionRadius=0
      boundBone {boneName=dyn_bow_c}
      boxExtents=<5 6 5>
      constraintSetup {angularLimitsMax=<2 2 1>, angularLimitsMin=<-1 -1 -1>, angularTarget=<0 0 0>, linearAxesMax=<1 0 0>}
      localJointOffset=<0 0 -1>
     }]
   }
   n7 = Anim Dynamics - Bone: dyn_bow_l(ComponentPose=@n6, Alpha=0.5, ExternalForce="0,0,0") {
    angularDampingOverride=20
    angularSpringConstant=100
    bAngularSpring=true
    bGravityOverrideInSimSpace=true
    bLinearSpring=true
    linearDampingOverride=20
    linearSpringConstant=180
    numSolverIterationsPostUpdate=4
    numSolverIterationsPreUpdate=16
    alphaBoolBlend {blendOption=Linear}
    boundBone {boneName=dyn_bow_l}
    componentAppliedLinearAccClamp=<2500 2500 2500>
    componentLinearAccScale=<1 1 1>
    componentLinearVelScale=<1 1 1>
    physicsBodyDefinitions [{
      sphereCollisionRadius=0
      boundBone {boneName=dyn_bow_l}
      boxExtents=<10 10 5>
      constraintSetup {angularLimitsMax=<2 10 10>, angularLimitsMin=<-2 -5 -10>, angularTarget=<0 0 0>, linearAxesMax=<0.5 2 10>, linearAxesMin=<-0.5 -3 -1>}
      localJointOffset=<0 0 -1>
     }]
   }
   n8 = Anim Dynamics - Bone: dyn_bow_r(ComponentPose=@n7, Alpha=0.5, ExternalForce="0,0,0") {
    angularDampingOverride=20
    angularSpringConstant=100
    bAngularSpring=true
    bGravityOverrideInSimSpace=true
    bLinearSpring=true
    linearDampingOverride=20
    linearSpringConstant=180
    numSolverIterationsPostUpdate=4
    numSolverIterationsPreUpdate=16
    alphaBoolBlend {blendOption=Linear}
    boundBone {boneName=dyn_bow_r}
    componentAppliedLinearAccClamp=<2500 2500 2500>
    componentLinearAccScale=<1 1 1>
    componentLinearVelScale=<1 1 1>
    physicsBodyDefinitions [{
      sphereCollisionRadius=0
      boundBone {boneName=dyn_bow_r}
      boxExtents=<10 10 5>
      constraintSetup {angularLimitsMax=<5 10 10>, angularLimitsMin=<-5 -5 -10>, angularTarget=<0 0 0>, linearAxesMax=<0.5 2 10>, linearAxesMin=<-0.5 -3 -1>}
      localJointOffset=<0 0 -1>
     }]
   }
   n9 = Anim Dynamics - Bone: dyn_foliage_t(ComponentPose=@n8, Alpha=0.35, ExternalForce="0,0,0") {
    angularSpringConstant=50
    bAngularSpring=true
    bGravityOverrideInSimSpace=true
    bLinearSpring=true
    linearSpringConstant=100
    numSolverIterationsPostUpdate=4
    numSolverIterationsPreUpdate=16
    alphaBoolBlend {blendOption=Linear}
    boundBone {boneName=dyn_foliage_t}
    componentAppliedLinearAccClamp=<2500 2500 2500>
    componentLinearAccScale=<1 1 1>
    componentLinearVelScale=<1 1 1>
    physicsBodyDefinitions [{
      sphereCollisionRadius=0
      boundBone {boneName=dyn_foliage_t}
      boxExtents=<5 2 5>
      constraintSetup {angularLimitsMax=<2 2 1>, angularLimitsMin=<-1 -1 -1>, angularTarget=<0 0 0>, linearAxesMax=<1 1 10>, linearAxesMin=<-1 -2 -1.5>}
      localJointOffset=<0 0 -2>
     }]
   }
   n10 = Anim Dynamics - Bone: dyn_foliage_l(ComponentPose=@n9, Alpha=0.35, ExternalForce="0,0,0") {
    angularSpringConstant=50
    bAngularSpring=true
    bGravityOverrideInSimSpace=true
    bLinearSpring=true
    linearSpringConstant=100
    numSolverIterationsPostUpdate=4
    numSolverIterationsPreUpdate=16
    alphaBoolBlend {blendOption=Linear}
    boundBone {boneName=dyn_foliage_l}
    componentAppliedLinearAccClamp=<2500 2500 2500>
    componentLinearAccScale=<1 1 1>
    componentLinearVelScale=<1 1 1>
    physicsBodyDefinitions [{
      sphereCollisionRadius=0
      boundBone {boneName=dyn_foliage_l}
      boxExtents=<5 5 5>
      constraintSetup {angularLimitsMax=<2 2 1>, angularLimitsMin=<-1 -1 -1>, angularTarget=<0 0 0>, linearAxesMax=<2 1 10>, linearAxesMin=<-1 -1 -1.5>}
      localJointOffset=<0 0 -1>
     }]
   }
   n11 = Anim Dynamics - Bone: dyn_foliage_r(ComponentPose=@n10, Alpha=0.35, ExternalForce="0,0,0") {
    angularSpringConstant=50
    bAngularSpring=true
    bGravityOverrideInSimSpace=true
    bLinearSpring=true
    linearSpringConstant=75
    numSolverIterationsPostUpdate=4
    numSolverIterationsPreUpdate=16
    alphaBoolBlend {blendOption=Linear}
    boundBone {boneName=dyn_foliage_r}
    componentAppliedLinearAccClamp=<2500 2500 2500>
    componentLinearAccScale=<1 1 1>
    componentLinearVelScale=<1 1 1>
    physicsBodyDefinitions [{
      sphereCollisionRadius=0
      boundBone {boneName=dyn_foliage_r}
      boxExtents=<5 5 5>
      constraintSetup {angularLimitsMax=<2 2 1>, angularLimitsMin=<-1 -1 -1>, angularTarget=<0 0 0>, linearAxesMax=<1 1.5 1.5>, linearAxesMin=<-1 -1.5 -1>}
      localJointOffset=<0 0 -1>
     }]
   }
   n12 = "Transform (Modify) Bone - Bone: dyn_bow_c"!(ComponentPose=@n11, bAlphaBoolEnabled:bool=True, Translation="0,-1.5,1", Rotation="0,0,25", Scale="1,1,1") -> (Pose) {
    alphaInputType=Bool
    rotationMode=BMM_Additive
    translationMode=BMM_Additive
    alphaBoolBlend {blendOption=Linear}
    boneToModify {boneName=dyn_bow_c}
    rotation {roll=25}
    translation {y=-1.5, z=1}
   }
   n13 = "Transform (Modify) Bone - Bone: dyn_bow_btm_l_01"!(ComponentPose=@n12, bAlphaBoolEnabled:bool=True, Translation="0,0,6", Rotation="0.0,0,35", Scale="1,1,1") -> (Pose) {
    alphaInputType=Bool
    rotationMode=BMM_Additive
    translationMode=BMM_Additive
    alphaBoolBlend {blendOption=Linear}
    boneToModify {boneName=dyn_bow_btm_l_01}
    rotation {roll=35}
    translation {z=6}
   }
   n14 = "Transform (Modify) Bone - Bone: dyn_bow_btm_r_01"!(ComponentPose=@n13, bAlphaBoolEnabled:bool=True, Translation="0,0,6", Rotation="0.0,0,35", Scale="1,1,1") -> (Pose) {
    alphaInputType=Bool
    rotationMode=BMM_Additive
    translationMode=BMM_Additive
    alphaBoolBlend {blendOption=Linear}
    boneToModify {boneName=dyn_bow_btm_r_01}
    rotation {roll=35}
    translation {z=6}
   }
   Output Pose(Result=Component To Local(ComponentPose=@n14)) {layerGroup=None}
  ]}
 EventGraph {nodes [
   TryGetPawnOwner()
   Event BlueprintUpdateAnimation()
  ]}
]
variables []
